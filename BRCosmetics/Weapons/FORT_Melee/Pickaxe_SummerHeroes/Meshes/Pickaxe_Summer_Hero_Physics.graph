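# Physics asset for the Summer Hero pickaxe mesh, a two-handed melee weapon with a dynamic rope piece.
# It has two capsule bodies, a long one along the handle bone and a short one on the dyn_rope_01 bone.
# A single constraint joins the rope bone to the handle with equal angular limits on both swing axes and twist, bounding how far the rope can move relative to the handle.
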PhysicsAsset /BRCosmetics/Weapons/FORT_Melee/Pickaxe_SummerHeroes/Meshes/Pickaxe_Summer_Hero_Physics
bodies {
 Melee_TwoHanded_Handle [{shapeType=Capsule, length=109.4413, radius=2.5126, center=<-0.0199 -0.7614 32.1927>, rotation=<rot 0 -90 0>}]
 dyn_rope_01 [{shapeType=Capsule, length=7, radius=3.25, center=<-0.5447 1.1996 9.7228>, rotation=<rot 11.7052 20.934 57.7023>}]
}
constraints [{bone1Name=dyn_rope_01, bone2Name=Melee_TwoHanded_Handle, swing1LimitDegrees=45, swing2LimitDegrees=45, twistLimitDegrees=45}]
